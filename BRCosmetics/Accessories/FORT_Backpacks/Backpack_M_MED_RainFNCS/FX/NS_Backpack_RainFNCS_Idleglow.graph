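# Niagara idle glow effect for the RainFNCS backpack cosmetic, built from a few burst-spawned sprites at fixed offsets.
# GlowEdges places six elongated, aligned sprites and SoftGlow two larger halos, using partition-indexed arrays for layout and two-tone magenta and orange coloring.
# Camera-vector dot products drive view-dependent alpha, with SoftGlow also scaled by a system oscillation value.
# GlowEdges001 and SoftGlow001 are disabled variants, the latter reading a system GlowCycle color.
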NiagaraSystem /BRCosmetics/Accessories/FORT_Backpacks/Backpack_M_MED_RainFNCS/FX/NS_Backpack_RainFNCS_Idleglow
dependencies {
 usedDataInterfaces [
  &NiagaraDataInterfaceArrayFloat2
  &NiagaraDataInterfaceArrayFloat3
  &NiagaraDataInterfaceCamera
  &NiagaraDataInterfaceArrayInt32
  &NiagaraDataInterfaceArrayFloat
  &NiagaraDataInterfaceSocketReader
  &NiagaraDataInterfaceColorCurve
 ]
 usedDynamicInputs [&ColorFromCurve, &Add_Float, &Sine]
 usedModules [
  &self:SystemSpawnScript.SystemScriptSource.SystemScriptGraph.NiagaraNodeAssignment_0.SetVariables_#1
  &SystemState
  &self:SystemSpawnScript.SystemScriptSource.SystemScriptGraph.NiagaraNodeAssignment_1.SetVariables_#2
  &self:SystemSpawnScript.SystemScriptSource.SystemScriptGraph.NiagaraNodeAssignment_2.SetVariables_#3
 ]
}
emitters {
 GlowEdges {
  emitterUpdateScript [
   &EmitterState(
    Inactive Response=::NewEnumerator1
    Life Cycle Mode=::NewEnumerator1
    Loop Behavior=::NewEnumerator1
    Loop Duration Mode=::NewEnumerator1
    UseLoopDelay=false
   )
   &SpawnBurst_Instantaneous(Spawn Count=6, Use Spawn Probability=false, UseLoopCountLimit=false)
  ]
  particleSpawnScript [
   &PartitionParticles(ID to Partition:NiagaraInt32=$Engine.ExecIndex, Number Of Partitions=6)
   &InitializeParticle(
    Color=<rgba 0.25 0.25 0.25 1>
    Sprite Size=&SelectVector2DFromArray(
     Direct Array Index=$Particles.PartitionParticles.Partition
     Vector 2D Selection Array=DataInterface() {propertyValues {FloatData [<9 22>, <9 22>, <8 17>, <8 17>, <8 20>, <8 20>]}}
    )
    Sprite Size Mode=::NewEnumerator4
    UsePositionOffset=false
    Write Material Random=true
   )
   /SetVariables(Particles.SpriteAlignment=&SelectVectorFromArray(
     Direct Array Index=$Particles.PartitionParticles.Partition
     Vector Selection Array=DataInterface() {propertyValues {FloatData [<1.5 -0.8 -0.5>, <1.5 0.8 0.5>, <-0.65 0.3 -0.95>, <0.65 0.3 -0.95>, <-0.7 0.3 -0.95>, <0.7 0.3 -0.95>]}}
    )) {isSetParameters=true, module=SetVariables_#4}
   /SetVariables(Particles.Position=&SelectVectorFromArray(
     Direct Array Index=$Particles.PartitionParticles.Partition
     Vector Selection Array=DataInterface() {propertyValues {FloatData [<5.5 -5 9.5>, <-5.5 -5 9.5>, <3 -3.5 -4>, <-3 -3.5 -4>, <8 -6 2>, <-8 -6 2>]}}
    )) {isSetParameters=true, module=SetVariables_#5}
  ]
  particleUpdateScript [
   &CameraOffset(Camera Offset Amount=&Lerp_Float(
     A=-5
     B=1
     Alpha=&NormalizeFloat(
      Value=&DotProduct(Vector A=<0 1 0>, Vector B=&ReturnCameraVector(Camera Vectors=::NewEnumerator1, Coordinate Space=::Simulation))
      Min=0
      Max=0
     )
    )) {enabled=false}
   &ParticleState(Kill Particles When Lifetime Has Elapsed=false, Let Infinitely Lived Particles Die When Emitter Deactivates=true)
   &ScaleColor(
    Scale Alpha=&Multiply_Float(
     A=1
     B=&Lerp_Float(
      A=0.5
      B=1
      Alpha=&DotProduct(Vector A=<0 1 0>, Vector B=&ReturnCameraVector(Camera Vectors=::NewEnumerator1, Coordinate Space=::Simulation))
     )
    )
    Scale RGB=&MaskVectorBySpawnGroup(
     Masked Vector=&MakeVectorFromLinearColorRGB(Color=<rgba 2.5 0 2.0261 1>)
     Passthrough Vector=&MakeVectorFromLinearColorRGB(Color=<rgba 1.5 0.525 0 1>)
     Spawn Group Mask=&SelectIntFromArray(Direct Array Index=$Particles.PartitionParticles.Partition, Int Selection Array=DataInterface() {propertyValues {IntData [1,1,0,0,1,1]}})
    )
    ScaleA=true
    ScaleRGB=true
   )
   &CameraOffset(Camera Offset Amount=3) {module=CameraOffset001}
  ]
  rendererClasses [&NiagaraSpriteRendererProperties]
  renderers [{rendererClass=&NiagaraSpriteRendererProperties}]
 }
 GlowEdges001 {
  bEnabled=false
  emitterUpdateScript [
   &EmitterState(
    Inactive Response=::NewEnumerator1
    Life Cycle Mode=::NewEnumerator1
    Loop Behavior=::NewEnumerator1
    Loop Duration Mode=::NewEnumerator1
    UseLoopDelay=false
   )
   &SpawnBurst_Instantaneous(Spawn Count=4, Use Spawn Probability=false, UseLoopCountLimit=false)
  ]
  particleSpawnScript [
   &PartitionParticles(ID to Partition:NiagaraInt32=$Engine.ExecIndex, Number Of Partitions=4)
   &InitializeParticle(
    Color=<rgba 0.25 0.25 0.25 1>
    Sprite Size=&SelectVector2DFromArray(
     Direct Array Index=$Particles.PartitionParticles.Partition
     Vector 2D Selection Array=DataInterface() {propertyValues {FloatData [<9 20>, <9 20>, <8 17>, <8 17>]}}
    )
    Sprite Size Mode=::NewEnumerator4
    UsePositionOffset=false
    Write Material Random=true
   )
   /SetVariables(Particles.SpriteAlignment=&SelectVectorFromArray(
     Direct Array Index=$Particles.PartitionParticles.Partition
     Vector Selection Array=DataInterface() {propertyValues {FloatData [<1.3 -0.8 -0.5>, <1.3 0.8 0.5>, <-0.65 0.3 -0.95>, <0.65 0.3 -0.95>]}}
    )) {isSetParameters=true, module=SetVariables_#4}
   /SetVariables(Particles.Position=&SelectVectorFromArray(
     Direct Array Index=$Particles.PartitionParticles.Partition
     Vector Selection Array=DataInterface() {propertyValues {FloatData [<5 -5.5 8>, <-5 -5.5 8>, <3 -3.5 -4>, <-3 -3.5 -4>]}}
    )) {isSetParameters=true, module=SetVariables_#5}
  ]
  particleUpdateScript [
   &CameraOffset(Camera Offset Amount=&Lerp_Float(
     A=-5
     B=1
     Alpha=&NormalizeFloat(
      Value=&DotProduct(Vector A=<0 1 0>, Vector B=&ReturnCameraVector(Camera Vectors=::NewEnumerator1, Coordinate Space=::Simulation))
      Min=0
      Max=1
     )
    ))
   &ParticleState(Kill Particles When Lifetime Has Elapsed=false, Let Infinitely Lived Particles Die When Emitter Deactivates=true)
   &ScaleColor(
    Scale Alpha=&Multiply_Float(
     A=1
     B=&Lerp_Float(
      A=0.5
      B=1
      Alpha=&DotProduct(Vector A=<0 1 0>, Vector B=&ReturnCameraVector(Camera Vectors=::NewEnumerator1, Coordinate Space=::Simulation))
     )
    )
    Scale RGB=&MaskVectorBySpawnGroup(
     Masked Vector=&MakeVectorFromLinearColorRGB(Color=<rgba 10 10 10 1>)
     Passthrough Vector=&MakeVectorFromLinearColorRGB(Color=<rgba 10 0 8.1044 1>)
     Spawn Group Mask=&MakeCustomIntFromBool(
      Bool=&SetBoolByIntComparison(A=2, B=$Particles.PartitionParticles.Partition, Comparison Type=::NewEnumerator7)
      True INT=1
      False INT=0
     )
    )
    ScaleA=true
    ScaleRGB=true
   )
  ]
  rendererClasses [&NiagaraSpriteRendererProperties]
  renderers [{rendererClass=&NiagaraSpriteRendererProperties}]
 }
 SoftGlow {
  emitterUpdateScript [
   &EmitterState(Life Cycle Mode=::NewEnumerator1, Loop Behavior=::NewEnumerator1, Loop Duration Mode=::NewEnumerator1, UseLoopDelay=false)
   &SpawnBurst_Instantaneous(Spawn Count=2, Use Spawn Probability=false, UseLoopCountLimit=false)
  ]
  particleSpawnScript [
   &PartitionParticles()
   &InitializeParticle(
    Color=<rgba 0.2 0.2 0.2 1>
    Lifetime=5
    Position Offset=&SelectVectorFromArray(
     Direct Array Index=$Particles.PartitionParticles.Partition
     Vector Selection Array=DataInterface() {propertyValues {FloatData [<0 -5 3>, <0 -3 -2>]}}
    )
    Sprite Size Mode=::NewEnumerator3
    Uniform Sprite Size=&SelectFloatFromArray(Direct Array Index=$Particles.PartitionParticles.Partition, Float Selection Array=DataInterface() {propertyValues {FloatData [44,22]}})
    UsePositionOffset=true
    Write Material Random=true
   )
  ]
  particleUpdateScript [
   &CameraOffset(Camera Offset Amount=&Lerp_Float(
     A=0
     B=0
     Alpha=&NormalizeFloat(
      Value=&DotProduct(Vector A=<0 1 0>, Vector B=&ReturnCameraVector(Camera Vectors=::NewEnumerator1, Coordinate Space=::Simulation))
      Min=1
      Max=1
     )
    )) {enabled=false, module=CameraOffset001}
   &ParticleState(Kill Particles When Lifetime Has Elapsed=false, Let Infinitely Lived Particles Die When Emitter Deactivates=true)
   &ScaleColor(
    Scale Alpha=&Lerp_Float(
     A=0
     B=$System.Oscillation
     Alpha=&Power(
      a=2
      b=&DotProduct(Vector A=<0 1 0>, Vector B=&ReturnCameraVector(Camera Vectors=::NewEnumerator1, Coordinate Space=::Simulation))
     )
    )
    Scale RGB=&MaskVectorBySpawnGroup(
     Masked Vector=&MakeVectorFromLinearColorRGB(Color=<rgba 1 0.35 0 1>)
     Passthrough Vector=&MakeVectorFromLinearColorRGB(Color=<rgba 0.75 0 0.6078 1>)
     Spawn Group Mask=$Particles.PartitionParticles.Partition
    )
    ScaleA=true
    ScaleRGB=true
   )
   &CameraOffset(Camera Offset Amount=16)
  ]
  rendererClasses [&NiagaraSpriteRendererProperties]
  renderers [{rendererClass=&NiagaraSpriteRendererProperties}]
 }
 SoftGlow001 {
  bEnabled=false
  emitterUpdateScript [
   &EmitterState(Life Cycle Mode=::NewEnumerator1, Loop Behavior=::NewEnumerator1, Loop Duration Mode=::NewEnumerator1, UseLoopDelay=false)
   &SpawnBurst_Instantaneous(Spawn Count=2, Use Spawn Probability=false, UseLoopCountLimit=false)
  ]
  particleSpawnScript [
   &PartitionParticles()
   &InitializeParticle(
    Color=<rgba 0.2 0.2 0.2 1>
    Lifetime=5
    Position Offset=&SelectVectorFromArray(
     Direct Array Index=$Particles.PartitionParticles.Partition
     Vector Selection Array=DataInterface() {propertyValues {FloatData [<0 -9.2 2.8>, <0 -5.2 -4>]}}
    )
    Sprite Size Mode=::NewEnumerator3
    Uniform Sprite Size=&SelectFloatFromArray(Direct Array Index=$Particles.PartitionParticles.Partition, Float Selection Array=DataInterface() {propertyValues {FloatData [28,18]}})
    UsePositionOffset=true
    Write Material Random=true
   )
  ]
  particleUpdateScript [
   &CameraOffset(Camera Offset Amount=&Lerp_Float(
     A=-5
     B=0
     Alpha=&NormalizeFloat(
      Value=&DotProduct(Vector A=<0 1 0>, Vector B=&ReturnCameraVector(Camera Vectors=::NewEnumerator1, Coordinate Space=::Simulation))
      Min=0
      Max=1
     )
    )) {module=CameraOffset001}
   &ParticleState(Kill Particles When Lifetime Has Elapsed=false, Let Infinitely Lived Particles Die When Emitter Deactivates=true)
   &ScaleColor(
    Scale Alpha=&Lerp_Float(
     A=0
     B=1
     Alpha=&Power(
      a=2
      b=&DotProduct(Vector A=<0 1 0>, Vector B=&ReturnCameraVector(Camera Vectors=::NewEnumerator1, Coordinate Space=::Simulation))
     )
    )
    Scale RGB=&MaskVectorBySpawnGroup(
     Masked Vector=&MakeVectorFromLinearColorRGB(Color=$System.GlowCycle)
     Passthrough Vector=&MakeVectorFromLinearColorRGB(Color=<rgba 0.5 0 0.4052 1>)
     Spawn Group Mask=$Particles.PartitionParticles.Partition
    )
    ScaleA=true
    ScaleRGB=true
   )
  ]
  rendererClasses [&NiagaraSpriteRendererProperties]
  renderers [{rendererClass=&NiagaraSpriteRendererProperties}]
 }
}
summary {
 systemName=NS_Backpack_RainFNCS_Idleglow
 emitters [
  {emitterName=GlowEdges, rendererClasses [&NiagaraSpriteRendererProperties]}
  {emitterName=SoftGlow, rendererClasses [&NiagaraSpriteRendererProperties]}
  {bEnabled=false, emitterName=GlowEdges001, rendererClasses [&NiagaraSpriteRendererProperties]}
  {bEnabled=false, emitterName=SoftGlow001, rendererClasses [&NiagaraSpriteRendererProperties]}
 ]
 userVariables [
  User.ColorCycle {type=NiagaraFloat, defaultValue=0, description=""}
  User.Reactivity {type=NiagaraBool}
 ]
}
