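# Control Rig for the Rain FNCS backpack cosmetic that orients its mirrored left and right five-bone dyn_pack chains.
# One forward pass runs local-space-offset aim constraints, alternating sides outward along the chains, with Z as the up direction.
# The first bone on each side partially aims at its dyn_pack_1_d bone under dyn_pack_element.
# Each later bone fully aims at the preceding one, linking bones that are siblings in the hierarchy.
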ControlRigBlueprint /BRCosmetics/Accessories/FORT_Backpacks/Backpack_M_MED_RainFNCS/Meshes/M_MED_RainFNCS_Pack_CtrlRig
graphs [
 RigVMFunctionLibrary {links [], nodes []}
 RigVMModel {nodes [
   RigUnit_BeginExecution = &RigUnit_BeginExecution()
   &RigUnit_AimConstraintLocalSpaceOffset(
    Child=(Type=Bone,Name="dyn_pack_1_l")
    Filter=(bX=True,bY=True,bZ=True)
    Parents=((Item=(Type=Bone,Name="dyn_pack_1_d_l"),Weight=1))
    Weight=0.75
    WorldUp=(Target=(X=0,Y=0,Z=1),Kind=Direction,Space=(Type=None,Name="None"))
    ExecutePin=@RigUnit_BeginExecution.ExecutePin
   )
   > &RigUnit_AimConstraintLocalSpaceOffset(
    Child=(Type=Bone,Name="dyn_pack_1_r")
    Filter=(bX=True,bY=True,bZ=True)
    Parents=((Item=(Type=Bone,Name="dyn_pack_1_d_r"),Weight=1))
    Weight=0.75
    WorldUp=(Target=(X=0,Y=0,Z=1),Kind=Direction,Space=(Type=None,Name="None"))
   )
   > &RigUnit_AimConstraintLocalSpaceOffset(
    Child=(Type=Bone,Name="dyn_pack_2_l")
    Filter=(bX=True,bY=True,bZ=True)
    Parents=((Item=(Type=Bone,Name="dyn_pack_1_l"),Weight=1))
    Weight=1
    WorldUp=(Target=(X=0,Y=0,Z=1),Kind=Direction,Space=(Type=None,Name="None"))
   )
   > &RigUnit_AimConstraintLocalSpaceOffset(
    Child=(Type=Bone,Name="dyn_pack_2_r")
    Filter=(bX=True,bY=True,bZ=True)
    Parents=((Item=(Type=Bone,Name="dyn_pack_1_r"),Weight=1))
    Weight=1
    WorldUp=(Target=(X=0,Y=0,Z=1),Kind=Direction,Space=(Type=None,Name="None"))
   )
   > &RigUnit_AimConstraintLocalSpaceOffset(
    Child=(Type=Bone,Name="dyn_pack_3_l")
    Filter=(bX=True,bY=True,bZ=True)
    Parents=((Item=(Type=Bone,Name="dyn_pack_2_l"),Weight=1))
    Weight=1
    WorldUp=(Target=(X=0,Y=0,Z=1),Kind=Direction,Space=(Type=None,Name="None"))
   )
   > &RigUnit_AimConstraintLocalSpaceOffset(
    Child=(Type=Bone,Name="dyn_pack_3_r")
    Filter=(bX=True,bY=True,bZ=True)
    Parents=((Item=(Type=Bone,Name="dyn_pack_2_r"),Weight=1))
    Weight=1
    WorldUp=(Target=(X=0,Y=0,Z=1),Kind=Direction,Space=(Type=None,Name="None"))
   )
   > &RigUnit_AimConstraintLocalSpaceOffset(
    Child=(Type=Bone,Name="dyn_pack_4_l")
    Filter=(bX=True,bY=True,bZ=True)
    Parents=((Item=(Type=Bone,Name="dyn_pack_3_l"),Weight=1))
    Weight=1
    WorldUp=(Target=(X=0,Y=0,Z=1),Kind=Direction,Space=(Type=None,Name="None"))
   )
   > &RigUnit_AimConstraintLocalSpaceOffset(
    Child=(Type=Bone,Name="dyn_pack_4_r")
    Filter=(bX=True,bY=True,bZ=True)
    Parents=((Item=(Type=Bone,Name="dyn_pack_3_r"),Weight=1))
    Weight=1
    WorldUp=(Target=(X=0,Y=0,Z=1),Kind=Direction,Space=(Type=None,Name="None"))
   )
   > &RigUnit_AimConstraintLocalSpaceOffset(
    Child=(Type=Bone,Name="dyn_pack_5_l")
    Filter=(bX=True,bY=True,bZ=True)
    Parents=((Item=(Type=Bone,Name="dyn_pack_4_l"),Weight=1))
    Weight=1
    WorldUp=(Target=(X=0,Y=0,Z=1),Kind=Direction,Space=(Type=None,Name="None"))
   )
   > &RigUnit_AimConstraintLocalSpaceOffset(
    Child=(Type=Bone,Name="dyn_pack_5_r")
    Filter=(bX=True,bY=True,bZ=True)
    Parents=((Item=(Type=Bone,Name="dyn_pack_4_r"),Weight=1))
    Weight=1
    WorldUp=(Target=(X=0,Y=0,Z=1),Kind=Direction,Space=(Type=None,Name="None"))
   )
  ]}
]
hierarchy [
 root {loc=<0 0 0>, rot=<0 0 0 1>}
 pack {parent=root, loc=<0 -0.7316 0>, rot=<0 0 0 1>}
 dyn_pack_base {parent=pack, loc=<0 0 0>, rot=<0 0 0 1>}
 dyn_pack_element {parent=dyn_pack_base, loc=<0 0 0>, rot=<0 0 0 1>}
 dyn_pack_1_d_r {parent=dyn_pack_element, loc=<-1.293 -0.3921 -12.1777>, rot=<0 0.4063 0 -0.9137>}
 dyn_pack_1_d_l {parent=dyn_pack_element, loc=<1.2926 -0.3921 -12.1777>, rot=<0 -0.9158 0 0.4016>}
 dyn_pack_5_r {parent=dyn_pack_base, loc=<-10.184 -3.2803 -3.4057>, rot=<0 -0.543 0 -0.8397>}
 dyn_pack_4_r {parent=dyn_pack_base, loc=<-9.103 -2.4977 -5.6827>, rot=<0 -0.4192 0 -0.9079>}
 dyn_pack_3_r {parent=dyn_pack_base, loc=<-7.2283 -1.9356 -7.6568>, rot=<0 -0.3493 0 -0.937>}
 dyn_pack_2_r {parent=dyn_pack_base, loc=<-4.93 -1.3181 -9.1914>, rot=<0 -0.2131 0 -0.977>}
 dyn_pack_1_r {parent=dyn_pack_base, loc=<-2.2673 -1.0559 -10.1012>, rot=<0 -0.1271 0 -0.9919>}
 dyn_pack_1_l {parent=dyn_pack_base, loc=<2.2673 -1.0559 -10.1012>, rot=<0 -0.9903 0 -0.1388>}
 dyn_pack_2_l {parent=dyn_pack_base, loc=<4.93 -1.3181 -9.1914>, rot=<0 -0.9753 0 -0.221>}
 dyn_pack_3_l {parent=dyn_pack_base, loc=<7.228 -1.9356 -7.6568>, rot=<0 -0.9344 0 -0.3561>}
 dyn_pack_4_l {parent=dyn_pack_base, loc=<9.1026 -2.4977 -5.6827>, rot=<0 -0.9065 0 -0.4222>}
 dyn_pack_5_l {parent=dyn_pack_base, loc=<10.1838 -3.2803 -3.4057>, rot=<0 -0.851 0 -0.5252>}
]
