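# Animation Blueprint for the RainFNCS backpack, adding procedural secondary motion to its floating parts.
# AnimDynamics linear springs move the main element and two five-bone chains, each bone translation-constrained to the previous one, before a Control Rig pass.
# EventGraph layers staggered sine-wave rolls and a vertical bob, adjusting spring blends, offset or gravity for skydiving, sliding and crouching.
# The bob overwrites skydive and parachute offsets each update; crouch overrides it.
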AnimBlueprint /BRCosmetics/Accessories/FORT_Backpacks/Backpack_M_MED_RainFNCS/Meshes/M_MED_RainFNCS_Pack_AnimBP
target_skeleton=&/M_MED_RainFNCS_Pack_Skeleton
graphs [
 AnimGraph {nodes [
   n1 = Anim Dynamics - Bone: dyn_pack_1_d_r!(ComponentPose, Alpha=0.8, GravityOverride:Vector="0,0,0", ExternalForce="0,0,0") -> (Pose) {
    bLinearSpring=true
    bOverrideLinearDamping=true
    bUseGravityOverride=true
    linearDampingOverride=0.995
    linearSpringConstant=5
    simulationSpace=BoneRelative
    alphaBoolBlend {blendOption=Linear}
    boundBone {boneName=dyn_pack_1_d_r}
    componentAppliedLinearAccClamp=<50 50 50>
    componentLinearAccScale=<0.75 0.75 0.75>
    componentLinearVelScale=<0.5 0.5 0.5>
    gravityOverride {z=150}
    physicsBodyDefinitions [{
      boundBone {boneName=dyn_pack_1_d_r}
      boxExtents=<1000 900 1200>
      constraintSetup {
       twistAxis=AxisY
       angularLimitsMax=<0 0 0>
       angularLimitsMin=<0 0 0>
       angularTarget=<0 0 0>
       linearAxesMax=<0 1.5 0>
       linearAxesMin=<0 -1.5 0>
      }
      localJointOffset=<0 0 0>
     }]
    relativeSpaceBone {boneName=dyn_pack_base}
   }
   Anim Dynamics - Bone: dyn_pack_1_d_l!(ComponentPose=@n1, Alpha=0.8, GravityOverride:Vector="0,0,0", ExternalForce="0,0,0") -> (Pose) {
    bLinearSpring=true
    bOverrideLinearDamping=true
    bUseGravityOverride=true
    linearDampingOverride=0.995
    linearSpringConstant=5
    simulationSpace=BoneRelative
    alphaBoolBlend {blendOption=Linear}
    boundBone {boneName=dyn_pack_1_d_l}
    componentAppliedLinearAccClamp=<50 50 50>
    componentLinearAccScale=<0.75 0.75 0.75>
    componentLinearVelScale=<0.5 0.5 0.5>
    gravityOverride {z=150}
    physicsBodyDefinitions [{
      boundBone {boneName=dyn_pack_1_d_l}
      boxExtents=<1000 900 1200>
      constraintSetup {
       twistAxis=AxisY
       angularLimitsMax=<0 0 0>
       angularLimitsMin=<0 0 0>
       angularTarget=<0 0 0>
       linearAxesMax=<0 1.5 0>
       linearAxesMin=<0 -1.5 0>
      }
      localJointOffset=<0 0 0>
     }]
    relativeSpaceBone {boneName=dyn_pack_base}
   }
   n2 = "Transform (Modify) Bone - Bone: dyn_pack_base"!(ComponentPose=Local To Component(LocalPose=Input Pose()), Translation="0,-1,0", Rotation="0,0,0", Scale="1,1,1") -> (Pose) {
    rotationMode=BMM_Additive
    rotationSpace=BCS_BoneSpace
    translationMode=BMM_Additive
    alphaBoolBlend {blendOption=Linear}
    boneToModify {boneName=dyn_pack_base}
    translation {y=-1}
   }
   n3 = "Transform (Modify) Bone - Bone: dyn_pack_element"!(ComponentPose=@n2, Translation=Get Dyn_transform_Main() -> (Dyn_transform_Main:Vector), Rotation="0,0,0", Scale="1,1,1") -> (Pose) {
    rotationMode=BMM_Additive
    rotationSpace=BCS_BoneSpace
    translationMode=BMM_Additive
    alphaBoolBlend {blendOption=Linear}
    boneToModify {boneName=dyn_pack_element}
   }
   n4 = Anim Dynamics - Bone: dyn_pack_element!(ComponentPose=@n3, Alpha=Get Dyn_Alpha_main() -> (Dyn_Alpha_main:float), GravityOverride:Vector="0,0,-20", ExternalForce="0,0,0") -> (Pose) {
    bLinearSpring=true
    bOverrideLinearDamping=true
    bUseGravityOverride=true
    linearDampingOverride=0.9999
    linearSpringConstant=10
    simulationSpace=BoneRelative
    alphaBoolBlend {blendOption=Linear}
    boundBone {boneName=dyn_pack_element}
    componentAppliedLinearAccClamp=<30 30 30>
    componentLinearAccScale=<0.5 0.5 0.5>
    componentLinearVelScale=<0.5 0.5 0.5>
    gravityOverride {z=150}
    physicsBodyDefinitions [{
      boundBone {boneName=dyn_pack_element}
      boxExtents=<2000 1800 2500>
      constraintSetup {
       twistAxis=AxisY
       angularLimitsMax=<0 0 0>
       angularLimitsMin=<0 0 0>
       angularTarget=<0 0 0>
       linearAxesMax=<0 0 10>
       linearAxesMin=<0 -0.1 -5>
      }
      localJointOffset=<0 -20 0>
     }]
    relativeSpaceBone {boneName=dyn_pack_base}
   }
   n5 = Get Gravity Override()
   n6 = Anim Dynamics - Bone: dyn_pack_1_d_r(ComponentPose=@n4, Alpha=0.8, GravityOverride=@n5, ExternalForce="0,0,0") {
    bLinearSpring=true
    bOverrideLinearDamping=true
    bUseGravityOverride=true
    linearDampingOverride=0.995
    linearSpringConstant=5
    simulationSpace=BoneRelative
    alphaBoolBlend {blendOption=Linear}
    boundBone {boneName=dyn_pack_1_d_r}
    componentAppliedLinearAccClamp=<50 50 50>
    componentLinearAccScale=<0.75 0.75 0.75>
    componentLinearVelScale=<0.5 0.5 0.5>
    gravityOverride {z=150}
    physicsBodyDefinitions [{
      boundBone {boneName=dyn_pack_1_d_r}
      boxExtents=<1000 900 1200>
      constraintSetup {
       twistAxis=AxisY
       angularLimitsMax=<0 0 0>
       angularLimitsMin=<0 0 0>
       angularTarget=<0 0 0>
       linearAxesMax=<0 0.1 0>
       linearAxesMin=<0 -0.1 0>
      }
      localJointOffset=<0 0 0>
     }]
    relativeSpaceBone {boneName=dyn_pack_base}
   }
   n7 = Anim Dynamics - Bone: dyn_pack_1_d_l(ComponentPose=@n6, Alpha=0.8, GravityOverride=@n5, ExternalForce="0,0,0") {
    bLinearSpring=true
    bOverrideLinearDamping=true
    bUseGravityOverride=true
    linearDampingOverride=0.995
    linearSpringConstant=5
    simulationSpace=BoneRelative
    alphaBoolBlend {blendOption=Linear}
    boundBone {boneName=dyn_pack_1_d_l}
    componentAppliedLinearAccClamp=<50 50 50>
    componentLinearAccScale=<0.75 0.75 0.75>
    componentLinearVelScale=<0.5 0.5 0.5>
    gravityOverride {z=150}
    physicsBodyDefinitions [{
      boundBone {boneName=dyn_pack_1_d_l}
      boxExtents=<1000 900 1200>
      constraintSetup {
       twistAxis=AxisY
       angularLimitsMax=<0 0 0>
       angularLimitsMin=<0 0 0>
       angularTarget=<0 0 0>
       linearAxesMax=<0 0.1 0>
       linearAxesMin=<0 -0.1 0>
      }
      localJointOffset=<0 0 0>
     }]
    relativeSpaceBone {boneName=dyn_pack_base}
   }
   n8 = "Constraint - dyn_pack_1_r (T)"(ComponentPose=@n7, Alpha=1, ConstraintWeights_0=1) {
    alphaBoolBlend {blendOption=Linear}
    boneToModify {boneName=dyn_pack_1_r}
    constraintSetup [{transformType=Translation, targetBone {boneName=dyn_pack_1_d_r}}]
    constraintWeights [1]
   }
   n9 = Get Dyn_Alpha_Parts() -> (Dyn_Alpha_Parts:float)
   n10 = Anim Dynamics - Bone: dyn_pack_1_r(ComponentPose=@n8, Alpha=@n9, ExternalForce="0,0,0") {
    angularSpringConstant=1
    bLinearSpring=true
    bOverrideLinearDamping=true
    bUseGravityOverride=true
    linearDampingOverride=0.9
    linearSpringConstant=8
    simulationSpace=BoneRelative
    alphaBoolBlend {blendOption=Linear}
    boundBone {boneName=dyn_pack_1_r}
    componentAppliedLinearAccClamp=<1 1 1>
    componentLinearAccScale=<0.5 0.5 0.5>
    componentLinearVelScale=<0.1 0.1 0.1>
    gravityOverride {z=50}
    physicsBodyDefinitions [{
      boundBone {boneName=dyn_pack_1_r}
      boxExtents=<400 350 500>
      constraintSetup {
       twistAxis=AxisY
       angularLimitsMax=<0 0 0>
       angularLimitsMin=<0 0 0>
       angularTarget=<1 0 0>
       linearAxesMax=<0.5 2 0.5>
       linearAxesMin=<-0.5 -2 -0.5>
      }
      localJointOffset=<0 0 0>
     }]
    relativeSpaceBone {boneName=dyn_pack_base}
   }
   n11 = Get Dyn_rotate_1() -> (Dyn_rotate_1:Rotator)
   n12 = "Transform (Modify) Bone - Bone: dyn_pack_1_r"!(ComponentPose=@n10, Translation="0,0,0", Rotation=@n11, Scale="1,1,1") -> (Pose) {
    alpha=0.85
    rotationMode=BMM_Additive
    rotationSpace=BCS_BoneSpace
    translationMode=BMM_Additive
    alphaBoolBlend {blendOption=Linear}
    boneToModify {boneName=dyn_pack_1_r}
   }
   n13 = "Constraint - dyn_pack_1_l (T)"(ComponentPose=@n12, Alpha=1, ConstraintWeights_0=1) {
    alphaBoolBlend {blendOption=Linear}
    boneToModify {boneName=dyn_pack_1_l}
    constraintSetup [{transformType=Translation, targetBone {boneName=dyn_pack_1_d_l}}]
    constraintWeights [1]
   }
   n14 = Anim Dynamics - Bone: dyn_pack_1_l(ComponentPose=@n13, Alpha=@n9, ExternalForce="0,0,0") {
    angularSpringConstant=1
    bLinearSpring=true
    bOverrideLinearDamping=true
    bUseGravityOverride=true
    linearDampingOverride=0.9
    linearSpringConstant=8
    simulationSpace=BoneRelative
    alphaBoolBlend {blendOption=Linear}
    boundBone {boneName=dyn_pack_1_l}
    componentAppliedLinearAccClamp=<1 1 1>
    componentLinearAccScale=<0.5 0.5 0.5>
    componentLinearVelScale=<0.1 0.1 0.1>
    gravityOverride {z=50}
    physicsBodyDefinitions [{
      boundBone {boneName=dyn_pack_1_l}
      boxExtents=<400 350 500>
      constraintSetup {
       twistAxis=AxisY
       angularLimitsMax=<0 0 0>
       angularLimitsMin=<0 0 0>
       angularTarget=<1 0 0>
       linearAxesMax=<0.5 2 0.5>
       linearAxesMin=<-0.5 -2 -0.5>
      }
      localJointOffset=<0 0 0>
     }]
    relativeSpaceBone {boneName=dyn_pack_base}
   }
   n15 = "Transform (Modify) Bone - Bone: dyn_pack_1_l"!(ComponentPose=@n14, Translation="0,0,0", Rotation=@n11, Scale="1,1,1") -> (Pose) {
    alpha=0.85
    rotationMode=BMM_Additive
    rotationSpace=BCS_BoneSpace
    translationMode=BMM_Additive
    alphaBoolBlend {blendOption=Linear}
    boneToModify {boneName=dyn_pack_1_l}
   }
   n16 = "Constraint - dyn_pack_2_r (T)"(ComponentPose=@n15, Alpha=1, ConstraintWeights_0=1) {
    alphaBoolBlend {blendOption=Linear}
    boneToModify {boneName=dyn_pack_2_r}
    constraintSetup [{transformType=Translation, targetBone {boneName=dyn_pack_1_r}}]
    constraintWeights [1]
   }
   n17 = Anim Dynamics - Bone: dyn_pack_2_r(ComponentPose=@n16, Alpha=@n9, ExternalForce="0,0,0") {
    angularSpringConstant=1
    bLinearSpring=true
    bOverrideLinearDamping=true
    bUseGravityOverride=true
    linearDampingOverride=0.915
    linearSpringConstant=8
    simulationSpace=BoneRelative
    alphaBoolBlend {blendOption=Linear}
    boundBone {boneName=dyn_pack_2_r}
    componentAppliedLinearAccClamp=<1 1 1>
    componentLinearAccScale=<0.5 0.5 0.5>
    componentLinearVelScale=<0.1 0.1 0.1>
    gravityOverride {z=50}
    physicsBodyDefinitions [{
      boundBone {boneName=dyn_pack_2_r}
      boxExtents=<350 300 400>
      constraintSetup {
       twistAxis=AxisY
       angularLimitsMax=<0 0 0>
       angularLimitsMin=<0 0 0>
       angularTarget=<1 0 0>
       linearAxesMax=<0.5 3.5 0.5>
       linearAxesMin=<-0.5 -3.5 -0.5>
      }
      localJointOffset=<0 0 0>
     }]
    relativeSpaceBone {boneName=dyn_pack_base}
   }
   n18 = Get Dyn_rotate_2() -> (Dyn_rotate_2:Rotator)
   n19 = "Transform (Modify) Bone - Bone: dyn_pack_2_r"!(ComponentPose=@n17, Translation="0,0,0", Rotation=@n18, Scale="1,1,1") -> (Pose) {
    alpha=0.85
    rotationMode=BMM_Additive
    rotationSpace=BCS_BoneSpace
    translationMode=BMM_Additive
    alphaBoolBlend {blendOption=Linear}
    boneToModify {boneName=dyn_pack_2_r}
   }
   n20 = "Constraint - dyn_pack_2_l (T)"(ComponentPose=@n19, Alpha=1, ConstraintWeights_0=1) {
    alphaBoolBlend {blendOption=Linear}
    boneToModify {boneName=dyn_pack_2_l}
    constraintSetup [{transformType=Translation, targetBone {boneName=dyn_pack_1_l}}]
    constraintWeights [1]
   }
   n21 = Anim Dynamics - Bone: dyn_pack_2_l(ComponentPose=@n20, Alpha=@n9, ExternalForce="0,0,0") {
    angularSpringConstant=1
    bLinearSpring=true
    bOverrideLinearDamping=true
    bUseGravityOverride=true
    linearDampingOverride=0.915
    linearSpringConstant=8
    simulationSpace=BoneRelative
    alphaBoolBlend {blendOption=Linear}
    boundBone {boneName=dyn_pack_2_l}
    componentAppliedLinearAccClamp=<1 1 1>
    componentLinearAccScale=<0.5 0.5 0.5>
    componentLinearVelScale=<0.1 0.1 0.1>
    gravityOverride {z=50}
    physicsBodyDefinitions [{
      boundBone {boneName=dyn_pack_2_l}
      boxExtents=<350 300 400>
      constraintSetup {
       twistAxis=AxisY
       angularLimitsMax=<0 0 0>
       angularLimitsMin=<0 0 0>
       angularTarget=<1 0 0>
       linearAxesMax=<0.5 3.5 0.5>
       linearAxesMin=<-0.5 -3.5 -0.5>
      }
      localJointOffset=<0 0 0>
     }]
    relativeSpaceBone {boneName=dyn_pack_base}
   }
   n22 = "Transform (Modify) Bone - Bone: dyn_pack_2_l"!(ComponentPose=@n21, Translation="0,0,0", Rotation=@n18, Scale="1,1,1") -> (Pose) {
    alpha=0.85
    rotationMode=BMM_Additive
    rotationSpace=BCS_BoneSpace
    translationMode=BMM_Additive
    alphaBoolBlend {blendOption=Linear}
    boneToModify {boneName=dyn_pack_2_l}
   }
   n23 = "Constraint - dyn_pack_3_r (T)"(ComponentPose=@n22, Alpha=1, ConstraintWeights_0=1) {
    alphaBoolBlend {blendOption=Linear}
    boneToModify {boneName=dyn_pack_3_r}
    constraintSetup [{transformType=Translation, targetBone {boneName=dyn_pack_2_r}}]
    constraintWeights [1]
   }
   n24 = Anim Dynamics - Bone: dyn_pack_3_r(ComponentPose=@n23, Alpha=@n9, ExternalForce="0,0,0") {
    angularSpringConstant=1
    bLinearSpring=true
    bOverrideLinearDamping=true
    bUseGravityOverride=true
    linearDampingOverride=0.93
    linearSpringConstant=8
    simulationSpace=BoneRelative
    alphaBoolBlend {blendOption=Linear}
    boundBone {boneName=dyn_pack_3_r}
    componentAppliedLinearAccClamp=<1 1 1>
    componentLinearAccScale=<0.5 0.5 0.5>
    componentLinearVelScale=<0.1 0.1 0.1>
    gravityOverride {z=50}
    physicsBodyDefinitions [{
      boundBone {boneName=dyn_pack_3_r}
      boxExtents=<300 250 350>
      constraintSetup {
       twistAxis=AxisY
       angularLimitsMax=<0 0 0>
       angularLimitsMin=<0 0 0>
       angularTarget=<1 0 0>
       linearAxesMax=<0.5 3 0.5>
       linearAxesMin=<-0.5 -3 -0.5>
      }
      localJointOffset=<0 0 0>
     }]
    relativeSpaceBone {boneName=dyn_pack_base}
   }
   n25 = Get Dyn_rotate_3() -> (Dyn_rotate_3:Rotator)
   n26 = "Transform (Modify) Bone - Bone: dyn_pack_3_r"!(ComponentPose=@n24, Translation="0,0,0", Rotation=@n25, Scale="1,1,1") -> (Pose) {
    alpha=0.85
    rotationMode=BMM_Additive
    rotationSpace=BCS_BoneSpace
    translationMode=BMM_Additive
    alphaBoolBlend {blendOption=Linear}
    boneToModify {boneName=dyn_pack_3_r}
   }
   n27 = "Constraint - dyn_pack_3_l (T)"(ComponentPose=@n26, Alpha=1, ConstraintWeights_0=1) {
    alphaBoolBlend {blendOption=Linear}
    boneToModify {boneName=dyn_pack_3_l}
    constraintSetup [{transformType=Translation, targetBone {boneName=dyn_pack_2_l}}]
    constraintWeights [1]
   }
   n28 = Anim Dynamics - Bone: dyn_pack_3_l(ComponentPose=@n27, Alpha=@n9, ExternalForce="0,0,0") {
    angularSpringConstant=1
    bLinearSpring=true
    bOverrideLinearDamping=true
    bUseGravityOverride=true
    linearDampingOverride=0.93
    linearSpringConstant=8
    simulationSpace=BoneRelative
    alphaBoolBlend {blendOption=Linear}
    boundBone {boneName=dyn_pack_3_l}
    componentAppliedLinearAccClamp=<1 1 1>
    componentLinearAccScale=<0.5 0.5 0.5>
    componentLinearVelScale=<0.1 0.1 0.1>
    gravityOverride {z=50}
    physicsBodyDefinitions [{
      boundBone {boneName=dyn_pack_3_l}
      boxExtents=<300 250 350>
      constraintSetup {
       twistAxis=AxisY
       angularLimitsMax=<0 0 0>
       angularLimitsMin=<0 0 0>
       angularTarget=<1 0 0>
       linearAxesMax=<0.5 3 0.5>
       linearAxesMin=<-0.5 -3 -0.5>
      }
      localJointOffset=<0 0 0>
     }]
    relativeSpaceBone {boneName=dyn_pack_base}
   }
   n29 = "Transform (Modify) Bone - Bone: dyn_pack_3_l"!(ComponentPose=@n28, Translation="0,0,0", Rotation=@n25, Scale="1,1,1") -> (Pose) {
    alpha=0.85
    rotationMode=BMM_Additive
    rotationSpace=BCS_BoneSpace
    translationMode=BMM_Additive
    alphaBoolBlend {blendOption=Linear}
    boneToModify {boneName=dyn_pack_3_l}
   }
   n30 = "Constraint - dyn_pack_4_r (T)"(ComponentPose=@n29, Alpha=1, ConstraintWeights_0=1) {
    alphaBoolBlend {blendOption=Linear}
    boneToModify {boneName=dyn_pack_4_r}
    constraintSetup [{transformType=Translation, targetBone {boneName=dyn_pack_3_r}}]
    constraintWeights [1]
   }
   n31 = Anim Dynamics - Bone: dyn_pack_4_r(ComponentPose=@n30, Alpha=@n9, ExternalForce="0,0,0") {
    angularSpringConstant=1
    bLinearSpring=true
    bOverrideLinearDamping=true
    bUseGravityOverride=true
    linearDampingOverride=0.945
    linearSpringConstant=8
    simulationSpace=BoneRelative
    alphaBoolBlend {blendOption=Linear}
    boundBone {boneName=dyn_pack_4_r}
    componentAppliedLinearAccClamp=<1 1 1>
    componentLinearAccScale=<0.5 0.5 0.5>
    componentLinearVelScale=<0.1 0.1 0.1>
    gravityOverride {z=50}
    physicsBodyDefinitions [{
      boundBone {boneName=dyn_pack_4_r}
      boxExtents=<250 200 300>
      constraintSetup {
       twistAxis=AxisY
       angularLimitsMax=<0 0 0>
       angularLimitsMin=<0 0 0>
       angularTarget=<1 0 0>
       linearAxesMax=<0.5 3 0.5>
       linearAxesMin=<-0.5 -3 -0.5>
      }
      localJointOffset=<0 0 0>
     }]
    relativeSpaceBone {boneName=dyn_pack_base}
   }
   n32 = Get Dyn_rotate_4() -> (Dyn_rotate_4:Rotator)
   n33 = "Transform (Modify) Bone - Bone: dyn_pack_4_r"!(ComponentPose=@n31, Translation="0,0,0", Rotation=@n32, Scale="1,1,1") -> (Pose) {
    alpha=0.85
    rotationMode=BMM_Additive
    rotationSpace=BCS_BoneSpace
    translationMode=BMM_Additive
    alphaBoolBlend {blendOption=Linear}
    boneToModify {boneName=dyn_pack_4_r}
   }
   n34 = "Constraint - dyn_pack_4_l (T)"(ComponentPose=@n33, Alpha=1, ConstraintWeights_0=1) {
    alphaBoolBlend {blendOption=Linear}
    boneToModify {boneName=dyn_pack_4_l}
    constraintSetup [{transformType=Translation, targetBone {boneName=dyn_pack_3_l}}]
    constraintWeights [1]
   }
   n35 = Anim Dynamics - Bone: dyn_pack_4_l(ComponentPose=@n34, Alpha=@n9, ExternalForce="0,0,0") {
    angularSpringConstant=1
    bLinearSpring=true
    bOverrideLinearDamping=true
    bUseGravityOverride=true
    linearDampingOverride=0.945
    linearSpringConstant=8
    simulationSpace=BoneRelative
    alphaBoolBlend {blendOption=Linear}
    boundBone {boneName=dyn_pack_4_l}
    componentAppliedLinearAccClamp=<1 1 1>
    componentLinearAccScale=<0.5 0.5 0.5>
    componentLinearVelScale=<0.1 0.1 0.1>
    gravityOverride {z=50}
    physicsBodyDefinitions [{
      boundBone {boneName=dyn_pack_4_l}
      boxExtents=<250 200 300>
      constraintSetup {
       twistAxis=AxisY
       angularLimitsMax=<0 0 0>
       angularLimitsMin=<0 0 0>
       angularTarget=<1 0 0>
       linearAxesMax=<0.5 3 0.5>
       linearAxesMin=<-0.5 -3 -0.5>
      }
      localJointOffset=<0 0 0>
     }]
    relativeSpaceBone {boneName=dyn_pack_base}
   }
   n36 = "Transform (Modify) Bone - Bone: dyn_pack_4_l"!(ComponentPose=@n35, Translation="0,0,0", Rotation=@n32, Scale="1,1,1") -> (Pose) {
    alpha=0.85
    rotationMode=BMM_Additive
    rotationSpace=BCS_BoneSpace
    translationMode=BMM_Additive
    alphaBoolBlend {blendOption=Linear}
    boneToModify {boneName=dyn_pack_4_l}
   }
   n37 = "Constraint - dyn_pack_5_r (T)"(ComponentPose=@n36, Alpha=1, ConstraintWeights_0=1) {
    alphaBoolBlend {blendOption=Linear}
    boneToModify {boneName=dyn_pack_5_r}
    constraintSetup [{transformType=Translation, targetBone {boneName=dyn_pack_4_r}}]
    constraintWeights [1]
   }
   n38 = Anim Dynamics - Bone: dyn_pack_5_r(ComponentPose=@n37, Alpha=@n9, ExternalForce="0,0,0") {
    angularSpringConstant=1
    bLinearSpring=true
    bOverrideLinearDamping=true
    bUseGravityOverride=true
    linearDampingOverride=0.96
    linearSpringConstant=8
    simulationSpace=BoneRelative
    alphaBoolBlend {blendOption=Linear}
    boundBone {boneName=dyn_pack_5_r}
    componentAppliedLinearAccClamp=<1 1 1>
    componentLinearAccScale=<0.5 0.5 0.5>
    componentLinearVelScale=<0.1 0.1 0.1>
    gravityOverride {z=50}
    physicsBodyDefinitions [{
      boundBone {boneName=dyn_pack_5_r}
      boxExtents=<200 150 250>
      constraintSetup {
       twistAxis=AxisY
       angularLimitsMax=<0 0 0>
       angularLimitsMin=<0 0 0>
       angularTarget=<1 0 0>
       linearAxesMax=<0.5 3 0.5>
       linearAxesMin=<-0.5 -3 -0.5>
      }
      localJointOffset=<0 0 0>
     }]
    relativeSpaceBone {boneName=dyn_pack_base}
   }
   n39 = Get Dyn_rotate_5() -> (Dyn_rotate_5:Rotator)
   n40 = "Transform (Modify) Bone - Bone: dyn_pack_5_r"!(ComponentPose=@n38, Translation="0,0,0", Rotation=@n39, Scale="1,1,1") -> (Pose) {
    alpha=0.85
    rotationMode=BMM_Additive
    rotationSpace=BCS_BoneSpace
    translationMode=BMM_Additive
    alphaBoolBlend {blendOption=Linear}
    boneToModify {boneName=dyn_pack_5_r}
   }
   n41 = "Constraint - dyn_pack_5_l (T)"(ComponentPose=@n40, Alpha=1, ConstraintWeights_0=1) {
    alphaBoolBlend {blendOption=Linear}
    boneToModify {boneName=dyn_pack_5_l}
    constraintSetup [{transformType=Translation, targetBone {boneName=dyn_pack_4_l}}]
    constraintWeights [1]
   }
   n42 = Anim Dynamics - Bone: dyn_pack_5_l(ComponentPose=@n41, Alpha=@n9, ExternalForce="0,0,0") {
    angularSpringConstant=1
    bLinearSpring=true
    bOverrideLinearDamping=true
    bUseGravityOverride=true
    linearDampingOverride=0.96
    linearSpringConstant=8
    simulationSpace=BoneRelative
    alphaBoolBlend {blendOption=Linear}
    boundBone {boneName=dyn_pack_5_l}
    componentAppliedLinearAccClamp=<1 1 1>
    componentLinearAccScale=<0.5 0.5 0.5>
    componentLinearVelScale=<0.1 0.1 0.1>
    gravityOverride {z=50}
    physicsBodyDefinitions [{
      boundBone {boneName=dyn_pack_5_l}
      boxExtents=<200 150 250>
      constraintSetup {
       twistAxis=AxisY
       angularLimitsMax=<0 0 0>
       angularLimitsMin=<0 0 0>
       angularTarget=<1 0 0>
       linearAxesMax=<0.5 3 0.5>
       linearAxesMin=<-0.5 -3 -0.5>
      }
      localJointOffset=<0 0 0>
     }]
    relativeSpaceBone {boneName=dyn_pack_base}
   }
   n43 = "Transform (Modify) Bone - Bone: dyn_pack_5_l"!(ComponentPose=@n42, Translation="0,0,0", Rotation=@n39, Scale="1,1,1") -> (Pose) {
    alpha=0.85
    rotationMode=BMM_Additive
    rotationSpace=BCS_BoneSpace
    translationMode=BMM_Additive
    alphaBoolBlend {blendOption=Linear}
    boneToModify {boneName=dyn_pack_5_l}
   }
   n44 = Control Rig(Source=Component To Local(ComponentPose=@n43)) {
    alphaBoolBlend {blendOption=Linear}
    controlRigAssetReference {blueprintRigClass=&/M_MED_RainFNCS_Pack_CtrlRig.M_MED_RainFNCS_Pack_CtrlRig_C}
   }
   Output Pose(Result=@n44) {layerGroup=None}
  ]}
 EventGraph {nodes [
   Event BlueprintUpdateAnimation()
   > n1 = Sequence!(execute) -> (then_0, then_1, then_2:Exec, then_3:Exec, then_4:Exec)
   > n2 = Branch(Condition=Get bIsSkyDiving())
   > Set Dyn_Alpha_main(Dyn_Alpha_main:float=0.25) -> (Output_Get:float)
   > Set Dyn_Alpha_Parts(Dyn_Alpha_Parts:float=0.2) -> (Output_Get:float)
   > Set Dyn_transform_Main(Dyn_transform_Main:Vector="0,-2,0") -> (Output_Get:Vector)
   > Branch(Condition=Get bIsParachuteOpened())
   > Set Dyn_transform_Main(Dyn_transform_Main="0,-0.5,0")
   Set Dyn_Alpha_main(execute=@n2.else, Dyn_Alpha_main=0.85)
   > Set Dyn_Alpha_Parts(Dyn_Alpha_Parts=0.5)
   > Set Dyn_transform_Main(Dyn_transform_Main="0,-1,1")
   n3 = "Sin (Radians)"(A=float * float(A=float * float(A=1, B=1), B=GetRealTimeSeconds()))
   Set Dyn_transform_Main(execute=@n1.then_1, Dyn_transform_Main=MakeVector(Y=0, Z=float * float(A=@n3, B=1)))
   n4 = "Sin (Radians)"(A=float * float(A=float * float(A=0.9, B=0.9), B=0))
   Set Dyn_rotate(execute=@n1.then_2, Dyn_rotate:Rotator=MakeRotator(Roll=float * float(A=@n4, B=10))) -> (Output_Get:Rotator)
   n5 = GetRealTimeSeconds()
   n6 = "Sin (Radians)"(A=float * float(A=float * float(A=0.9, B=0.9), B=@n5))
   > Set Dyn_rotate_1(Dyn_rotate_1:Rotator=MakeRotator(Roll=float * float(A=@n6, B=10))) -> (Output_Get:Rotator)
   n7 = "Sin (Radians)"(A=float * float(A=float * float(A=1.04, B=1), B=@n5))
   > Set Dyn_rotate_2(Dyn_rotate_2:Rotator=MakeRotator(Roll=float * float(A=@n7, B=10))) -> (Output_Get:Rotator)
   n8 = "Sin (Radians)"(A=float * float(A=float * float(A=1.06, B=1), B=@n5))
   > Set Dyn_rotate_3(Dyn_rotate_3:Rotator=MakeRotator(Roll=float * float(A=@n8, B=10))) -> (Output_Get:Rotator)
   n9 = "Sin (Radians)"(A=float * float(A=float * float(A=1.095, B=1), B=@n5))
   > Set Dyn_rotate_4(Dyn_rotate_4:Rotator=MakeRotator(Roll=float * float(A=@n9, B=10))) -> (Output_Get:Rotator)
   n10 = "Sin (Radians)"(A=float * float(A=float * float(A=1.1, B=1), B=@n5))
   > Set Dyn_rotate_5(Dyn_rotate_5:Rotator=MakeRotator(Roll=float * float(A=@n10, B=10))) -> (Output_Get:Rotator)
   n11 = Branch(execute=@n1.then_4, Condition=Get bIsSliding())
   > Set Dyn_Alpha_Parts(Dyn_Alpha_Parts=0.33)
   n12 = Branch(execute=@n11.else, Condition=Get bIsCrouching())
   > Set Dyn_transform_Main(Dyn_transform_Main="0,1.5,0")
   > Set Gravity Override(Gravity Override="0,0,-50")
   > Set Dyn_Alpha_Transform(Dyn_Alpha_Transform:float=1) -> (Output_Get:float)
   Set Gravity Override(execute=@n12.else, Gravity Override="0,0,150")
   > Set Dyn_Alpha_Transform(Dyn_Alpha_Transform=0)
   TryGetPawnOwner()
  ]}
]
variables [
 Dyn_Alpha_main {type=number, default=1}
 Dyn_Alpha_Parts {type=number}
 Dyn_Alpha_Transform {type=number, default=1}
 Dyn_rotate {type=Rotator}
 Dyn_rotate_1 {type=Rotator}
 Dyn_rotate_2 {type=Rotator}
 Dyn_rotate_3 {type=Rotator}
 Dyn_rotate_4 {type=Rotator}
 Dyn_rotate_5 {type=Rotator}
 Dyn_transform {type=Vector}
 Dyn_transform_Main {type=Vector}
 Gravity Override {type=Vector}
 Gravity Override Params {
  type=GravityOverrideParamsStruct
  default {onGround_Moving {gravityOverride=<0 0 30>}, onGround_Standing {gravityOverride=<0 0 300>}, tactical_Sprint {gravityOverride=<100 100 200>}}
 }
]
